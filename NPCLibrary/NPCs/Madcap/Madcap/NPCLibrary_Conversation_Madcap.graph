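# Conversation database for the Madcap NPC from the NPC library, opened by the Conversation.Entry.Madcap tag.
# A single Speech node gives the greeting, swapping the default line for prioritized contextual ones based on time since last match, first-ever conversation, or Bars held.
# The speech then branches through link nodes into the shared NPC library Quests and BasicServices conversations.
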ConversationDatabase /NPCLibrary/NPCs/Madcap/Madcap/NPCLibrary_Conversation_Madcap
entry_points [{tag=Conversation.Entry.Madcap, destinations [#1]}]
nodes [
 ConversationEntryPointNode_7 {
  id=#1
  class=ConversationEntryPointNode
  outputs [#2]
  properties {entryTag {tagName=Conversation.Entry.Madcap}}
 }
 Speech_C_4 {
  id=#2
  class=Speech_C
  outputs [#3, #4]
  properties {
   nodeName=""
   generalConfig {
    defaultMessage=You being kind to the mushrooms?
    contextualMessages [
     {
      message="Welcome back! I am your new mushroom overlord... Ha, kidding."
      contextRequirements [{requirement {class=FortControllerRequirement_CheckTimeSinceLastPlayedMatch}}]
      priority {curve {rowName=Tandem.Speech.Priority.WhileSincePlayed}}
     }
     {
      message="You're a fun guy, I'm a fungi. Haha! That one kills at mushroom parties."
      contextRequirements [{requirement {class=FortControllerRequirement_IsFirstEverConversationWithNPC}}]
      priority {curve {rowName=Tandem.Speech.Priority.FirstEverInteraction}}
     }
     {
      message=You trying to cap your Bar spending?
      contextRequirements [{requirement {class=FortControllerRequirement_HasBars, properties {thresholdTestConfig {quantity {value=3000}}}}}]
      priority {curve {rowName=Tandem.Speech.Priority.HighTotalBars}}
     }
    ]
   }
  }
 }
 ConversationLinkNode_0 {id=#4, class=ConversationLinkNode, properties {remoteEntryTag {tagName=Conversation.Entry.BasicServices}}}
 ConversationLinkNode_1 {id=#3, class=ConversationLinkNode, properties {remoteEntryTag {tagName=Conversation.Entry.NPCLibrary.Quests}}}
]
